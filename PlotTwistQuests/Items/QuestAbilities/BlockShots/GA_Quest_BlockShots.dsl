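; GA_Quest_BlockShots is a quest gameplay ability that records blocked hits for a quest objective.
; On activation from an event it calls the parent, then sends a custom stat event with the BlockedHit tag and a count of one to the owning player's quest manager.
; It ends after a short delay, and its CanApplyHolster override always returns false.
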(event Ability|EventActivateAbilityFromEvent
  (params (EventData :type Gameplay%20Event%20Data%20Structure%20%28by%20ref%29))
  (|CallParentFunction EventData)
  (Quest|SendComplexCustomStatEvent (Variables|GAVars|Data|GetOwningPlayerQuestManager) 0 (Variables|Default|GetT_BlockedHitTag) "(GameplayTags=)" 1)
  (Utilities|FlowControl|Delay 0.2)
  (Ability|EndAbility))

(fn CanApplyHolster
  (params (HolsterId :type Name%20%28by%20ref%29 :default "None"))
  (return false))
